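# Additive unlit particle and Niagara sprite material for a lens flare on the NeonGlow rocket weapon effect.
# Color is either particle color or a fixed texel of the neon emissive texture, picked by a selection parameter and scaled by dynamic intensity, then fed through DeriveHDRfromLDR.
# Flare texture and particle alpha form opacity, scaled by time of day.
# Optional depth fade is bypassed on ES2 and ES3_1.
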Material /Figment_S10/Weapons/FORT_Special/NeonGlow/Materials/M_NeonRocket_LensFlare
graph {
 properties {
  bUsedWithNiagaraSprites=true
  bUsedWithParticleSprites=true
  blendMode=BLEND_Additive
  pixelDepthOffsetMode=PDOM_Legacy
  shadingModel=MSM_Unlit
  usedShadingModels=Unlit
 }
 nodes [
  n1 = ParticleColor()
  n2 = Constant() {r=1}
  n3 = ScalarParameter() {defaultValue=0.9, group=Neon, parameterName=Pulse_Color_Selection, sortPriority=2}
  n4 = Multiply(
   A=TextureSampleParameter2D(UVs=AppendVector(A=@n2, B=@n3)) {group=Neon, parameterName=Neon_Emissive, texture=/Game/Weapons/FORT_Rifles/Textures/Neon_Glow_GlowParts_E}.RGB
   B=Multiply(
    A=ScalarParameter() {defaultValue=5, parameterName=EmissiveIntensity}
    B=DynamicParameter() {paramNames [Overall_Intensity, Param2, Param3, Param4]}.Overall_Intensity
   )
  ) {constB=3}
  MaterialFunctionCall(
   LDR Input=StaticSwitchParameter(True=Clamp(None=@n1.RGB), False=Clamp(None=@n4)) {parameterName=UseParticleColor}
   HDR Tint=StaticSwitchParameter(True=@n1.RGB, False=@n4) {parameterName=UseParticleColor}
  ) {materialFunction=&DeriveHDRfromLDR}
  n5 = Multiply(
   A=ParticleColor().A
   B=TextureSampleParameter2D() {parameterName=LFTexture, samplerType=SAMPLERTYPE_Grayscale, texture=/Game/Effects/Fort_Effects/Textures/T_EFX_CameraFlare_01}.R
  )
  n6 = StaticSwitchParameter(True=DepthFade(Opacity=@n5, FadeDistance=ScalarParameter() {defaultValue=30, parameterName=FadeDistance}), False=@n5) {parameterName=UseDepthFade}
  Multiply(
   A=MaterialFunctionCall(
    Color_Morning=ScalarParameter() {defaultValue=1, parameterName=OpacityMorning}
    Color_Day=ScalarParameter() {defaultValue=1, parameterName=OpacityDay}
    Color_Evening=ScalarParameter() {defaultValue=1, parameterName=OpacityEvening}
    Color_Night=ScalarParameter() {defaultValue=1, parameterName=OpacityNight}
   ) {materialFunction=&Color_Time_of_Day}.Result
   B=FeatureLevelSwitch(Default=@n6, ES2=@n5, ES3_1=@n5, SM4_REMOVED=@n6)
  )
 ]
}
